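# Physics asset for the Football 20 Referee male body cosmetic, set up for a dangling whistle on a dynamic bone chain.
# Five small capsules along dyn_whistle_02 to 06 are linked by swing-limited, low-twist constraints and anchored to spine_05 with a tighter swing limit.
# Larger capsules on spine_03 to spine_05 form the torso volume, joined by free-motion constraints.
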PhysicsAsset /BRCosmetics/Characters/Player/Male/Medium/Bodies/M_MED_Football_20_Referee_HIS/Meshes/Parts/M_MED_Football_20_Referee_HIS_Physics
bodies {
 dyn_whistle_02 [{shapeType=Capsule, length=2, radius=3, shapeName=dyn_whistle_02_capsule, center=<4 0 2>, rotation=<rot 90 0 0>}]
 dyn_whistle_03 [{shapeType=Capsule, length=2, radius=3, shapeName=dyn_whistle_03_capsule, center=<4 0 2>, rotation=<rot 90 0 0>}]
 dyn_whistle_04 [{shapeType=Capsule, length=2, radius=3, shapeName=dyn_whistle_04_capsule, center=<4 0 2>, rotation=<rot 90 0 0>}]
 dyn_whistle_05 [{shapeType=Capsule, length=5, radius=1, shapeName=dyn_whistle_05_capsule, center=<4 0 0>, rotation=<rot 90 0 0>}]
 dyn_whistle_06 [{shapeType=Capsule, length=3, radius=2, shapeName=dyn_whistle_06_capsule, center=<4 0 0>, rotation=<rot 90 0 0>}]
 spine_03 [{shapeType=Capsule, length=5.2806, radius=13.7641, center=<3.1059 -0.087 0>}]
 spine_04 [{shapeType=Capsule, length=4.5831, radius=14.9607, center=<2.1462 -0.1638 0>}]
 spine_05 [
  {shapeType=Capsule, length=11.9007, radius=12.3874, center=<-1.2062 0.0064 0>}
  {shapeType=Capsule, length=17.4992, radius=5.2591, center=<-1.5342 8.097 -0.0001>}
  {shapeType=Capsule, length=13.2547, radius=9.6081, center=<2.5617 0 0>}
 ]
}
constraints [
 {bone1Name=dyn_whistle_06, bone2Name=dyn_whistle_05, swing1LimitDegrees=45, swing2LimitDegrees=45, twistLimitDegrees=5}
 {bone1Name=dyn_whistle_05, bone2Name=dyn_whistle_04, swing1LimitDegrees=45, swing2LimitDegrees=45, twistLimitDegrees=5}
 {
  bone1Name=spine_05
  bone2Name=spine_04
  swing1LimitDegrees=45
  swing1Motion=Free
  swing2LimitDegrees=45
  swing2Motion=Free
  twistLimitDegrees=45
  twistMotion=Free
 }
 {
  bone1Name=spine_04
  bone2Name=spine_03
  swing1LimitDegrees=45
  swing1Motion=Free
  swing2LimitDegrees=45
  swing2Motion=Free
  twistLimitDegrees=45
  twistMotion=Free
 }
 {bone1Name=dyn_whistle_02, bone2Name=spine_05, swing1LimitDegrees=10, swing2LimitDegrees=10, twistLimitDegrees=15}
 {bone1Name=dyn_whistle_03, bone2Name=dyn_whistle_02, swing1LimitDegrees=45, swing2LimitDegrees=45, twistLimitDegrees=5}
 {bone1Name=dyn_whistle_04, bone2Name=dyn_whistle_03, swing1LimitDegrees=45, swing2LimitDegrees=45, twistLimitDegrees=5}
]
