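# Animation Blueprint for the BLK Football referee male body part, adding secondary physics motion such as the whistle.
# It copies the attached parent's pose, then runs a RigidBody node in root-bone space using the HIS variant's physics asset.
# Update marks rigid body disabled when it is switched off or LOD exceeds 1.
# When that flag and the front-end flag are both true, an additive rotation adjusts dyn_whistle_05.
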AnimBlueprint /BRCosmetics/Characters/Player/Male/Medium/Bodies/M_MED_Football_20_Referee_BLK/Meshes/Parts/M_MED_Football_20_Referee_BLK_AnimBP
target_skeleton=&/M_MED_Football_20_Referee_BLK_Skeleton
graphs [
 AnimGraph {nodes [
   n1 = RigidBody(ComponentPose=Local To Component(LocalPose=Copy Pose From Mesh() {bUseAttachedParent=true})) {
    bClampLinearTranslationLimitToRefPose=true
    bOverrideWorldGravity=true
    overridePhysicsAsset=/BRCosmetics/Characters/Player/Male/Medium/Bodies/M_MED_Football_20_Referee_HIS/Meshes/Parts/M_MED_Football_20_Referee_HIS_Physics
    simulationSpace=BaseBoneSpace
    alphaBoolBlend {blendOption=Linear}
    baseBoneRef {boneName=root}
    componentAppliedLinearAccClamp=<300 300 300>
    componentLinearAccScale=<0.3 0.3 0.5>
    componentLinearVelScale=<0.2 0.2 0.3>
    overrideWorldGravity {z=-980}
   }
   n2 = "Transform (Modify) Bone - Bone: dyn_whistle_05"!(
    ComponentPose=@n1
    bAlphaBoolEnabled:bool=AND Boolean(A=Get isRigidBodyDisabled() -> (isRigidBodyDisabled:bool), B=Get bIsInFrontEnd())
    Translation="0,0,0"
    Rotation="-5,0,-15"
    Scale="1,1,1"
   ) -> (Pose) {
    alphaInputType=Bool
    rotationMode=BMM_Additive
    alphaBoolBlend {blendOption=Linear}
    boneToModify {boneName=dyn_whistle_05}
    rotation {pitch=-5, roll=-15}
   }
   Output Pose(Result=Component To Local(ComponentPose=@n2)) {layerGroup=None}
  ]}
 EventGraph {nodes [
   Event BlueprintUpdateAnimation()
   n1 = Get CurrentLOD()
   n2 = Get bRigidBodyEnabled()
   > Set isRigidBodyDisabled(isRigidBodyDisabled:bool=OR Boolean(A=NOT Boolean(A=@n2), B="integer > integer"(A=@n1, B=1) {is=title K2Node_CallFunction})) -> (Output_Get:bool)
  ]}
]
variables [isRigidBodyDisabled {type=boolean}]
